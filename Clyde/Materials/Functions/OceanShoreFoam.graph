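# Material function that layers shore foam onto ocean water material attributes, also outputting foam mask, color, normal and scatter mask.
# An exponential falloff of water depth and vertical world-position-offset displacement combines with world-XY foam and swirl textures, minus simulation and far-distance erasure.
# The result blends color, opacity, specular, roughness and normal.
# Low/medium quality and forward/mobile paths use a simpler depth-edge mask, flat color, and only set color and opacity.
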MaterialFunction /Clyde/Materials/Functions/OceanShoreFoam
graph {nodes [
  n1 = GetMaterialAttributes(None=FunctionInput() {bUsePreviewValueAsDefault=true, inputName=Attributes, inputType=FunctionInput_MaterialAttributes}) {attributeGetTypes [#1]}
  ComponentMask(None=@n1.WorldPositionOffset) {b=true, desc=WPO}
  n2 = FunctionInput() {bUsePreviewValueAsDefault=true, inputName=Mask, inputType=FunctionInput_Scalar, sortPriority=2, previewValue=<1 0 0 1>}
  n3 = ComponentMask(None=WorldPosition().XYZ) {b=true}
  n4 = ComponentMask(None=WorldPosition() {worldPositionShaderOffset=WPT_ExcludeAllShaderOffsets}.XYZ) {b=true}
  n5 = Multiply(A=FunctionInput() {bUsePreviewValueAsDefault=true, inputName=Depth, inputType=FunctionInput_Scalar, sortPriority=3}) {desc=Depth}
  MaterialExpressionAdd_6 = Add(
   A=Multiply(A=Subtract(A=@n3, B=@n4), B=ScalarParameter() {defaultValue=-0.03, group=Foam and Pondscum, parameterName="WPO Foam "})
   B=Multiply(A=@n5, B=ScalarParameter() {defaultValue=0.004, group=Foam and Pondscum, parameterName=depth foam amount})
  )
  n6 = Custom!(x=@MaterialExpressionAdd_6) {
   code="return exp(-x);"
   outputType=CMOT_Float1
   inputs [{inputName=x, input {expression=&self:MaterialExpressionAdd_6}}]
  }
  n7 = Subtract(A=@n6, B=ScalarParameter() {group=Foam and Pondscum, parameterName=Foam Dark Color Cutoff})
  n8 = Add(B=ComponentMask(None=WorldPosition() {worldPositionShaderOffset=WPT_ExcludeAllShaderOffsets}.XYZ) {g=true, r=true})
  n9 = TextureSample!(UVs=Divide(A=@n8, B=ScalarParameter() {defaultValue=2175, group=Foam and Pondscum, parameterName=FoamScale2}), Tex, Bias, Apply View MipBias) {mipValueMode=TMVM_MipBias, samplerSource=SSM_Wrap_WorldGroupSettings, texture=/FortWater/Textures/T_SwirlFlow_02_D}
  n10 = Multiply(A=@n9.R, B=ScalarParameter() {defaultValue=1, group=Foam and Pondscum, parameterName=FoamScatter Depth TEx})
  n11 = StaticSwitchParameter(True=Saturate(None=Multiply(A=@n7, B=@n10)), False=Constant()) {defaultValue=true, parameterName=Use Ocean Foam}
  n12 = VectorParameter() {group=Water Shading Model, parameterName=FoamScattering, defaultValue=<rgba 0 0.1 0.2 5>}
  n13 = Divide(A=@n12.A) {constB=100}
  FunctionOutput(None=Multiply(A=Multiply(A=@n2, B=@n11), B=Multiply(A=@n12.RGB, B=@n13))) {outputName=Scatter Mask, sortPriority=3}
  n14 = GetMaterialAttributes(None=@n1.MaterialAttributes) {attributeGetTypes [
    #2
    #3
    #4
    #5
    #6
   ]}
  n15 = VectorParameter() {group=Foam and Pondscum, parameterName=Foam Color 1, defaultValue=<rgba 0.2 0.7 1 0.4>}
  n16 = VectorParameter() {group=Foam and Pondscum, parameterName=Foam Color 2, defaultValue=<rgba 1 1 1 0.38>}
  n17 = Multiply(A=@n16.RGB, B=@n16.RGB)
  n18 = Subtract(A=@n6) {constB=0.5}
  n19 = ScalarParameter() {defaultValue=0.5, group=Foam and Pondscum, parameterName=Foam Foam Erode Min}
  n20 = ScalarParameter() {defaultValue=6, group=Foam and Pondscum, parameterName=Foam Texture 2 Height}
  n21 = Divide(A=@n8, B=ScalarParameter() {defaultValue=1250, group=Foam and Pondscum, parameterName=FoamScale})
  n22 = TextureSample(UVs=@n21) {
   mipValueMode=TMVM_MipBias
   samplerSource=SSM_Wrap_WorldGroupSettings
   samplerType=SAMPLERTYPE_LinearColor
   texture=/FortWater/Textures/T_foam_Normal2
  }
  n23 = Subtract(A=LinearInterpolate(A=Multiply(A=@n20, B=@n9.R), B=@n22.A, Alpha=ScalarParameter() {defaultValue=0.9, group=Foam and Pondscum, parameterName=FoamAlphaContrast})) {constB=0.5}
  n24 = ScalarParameter() {defaultValue=0.75, group=Foam and Pondscum, parameterName=Foam TexHeight}
  n25 = FunctionInput() {bUsePreviewValueAsDefault=true, inputName=Far Mask, inputType=FunctionInput_Scalar, sortPriority=2}
  n26 = ScalarParameter() {defaultValue=0.15, parameterName=Far Foam Remove}
  n27 = Subtract(
   A=Add(A=Min(A=@n18, B=@n19), B=Multiply(A=@n23, B=@n24))
   B=Add(
    A=FunctionInput() {bUsePreviewValueAsDefault=true, inputName=Sim Erasure, inputType=FunctionInput_Scalar, sortPriority=1}
    B=Multiply(A=@n25, B=@n26)
   )
  )
  n28 = LinearInterpolate(A=Multiply(A=@n15.RGB, B=@n15.RGB), B=@n17, Alpha=Saturate(None=@n27))
  n29 = QualitySwitch(Default=@n28, Low=@n17, High=@n28, Medium=@n17)
  n30 = LinearInterpolate(
   A=@n14.BaseColor
   B=FunctionOutput(None=ShadingPathSwitch(Default=@n29, Deferred=@n29, Forward=@n17, Mobile=@n17)) {outputName=Color, sortPriority=2}
   Alpha=@n2
  )
  n31 = Subtract(A=@n5, B=ScalarParameter() {defaultValue=16, parameterName=Depth Foam Offset})
  n32 = ScalarParameter() {defaultValue=16, group=Foam and Pondscum, parameterName=Depth for Foam Edge}
  n33 = ScalarParameter() {defaultValue=2, group=Foam and Pondscum, parameterName=Foam Sharpen}
  n34 = Multiply(A=Saturate(None=Divide(A=@n31, B=@n32)), B=Saturate(None=Multiply(A=@n27, B=@n33)))
  n35 = Divide(
   A=Subtract(A=@n5, B=ScalarParameter() {defaultValue=69.7077, parameterName=Depth Foam Offset_LOW})
   B=ScalarParameter() {defaultValue=300, group=Foam and Pondscum, parameterName=Depth for Foam Edge_LOW}
  )
  n36 = Multiply(A=OneMinus(None=Saturate(None=@n35)), B=Multiply(A=Subtract(A=@n22.A) {constB=0.5}, B=@n24))
  n37 = QualitySwitch(Default=@n34, Low=@n36, High=@n34, Medium=@n36)
  n38 = ShadingPathSwitch(Default=@n37, Deferred=@n37, Forward=@n36, Mobile=@n36)
  n39 = Constant()
  n40 = FunctionOutput(None=Multiply(A=StaticSwitchParameter(True=@n38, False=@n39) {defaultValue=true, parameterName=Use Ocean Foam}, B=@n2)) {outputName=Mask, sortPriority=1}
  n41 = LinearInterpolate(A=@n14.Opacity, B=ScalarParameter() {defaultValue=2, group=Foam and Pondscum, parameterName=Foam Opacity}, Alpha=@n40)
  n42 = ComponentMask(None=@n22.RGB) {g=true, r=true}
  n43 = Multiply(
   A=AppendVector(A=ConstantBiasScale(None=@n42) {bias=-0.5, scale=2}, B=@n22.B)
   B=VectorParameter() {parameterName=FoamNormal, defaultValue=<rgba 1 1 1 0>}.RGB
  )
  n44 = SetMaterialAttributes!(
   MaterialAttributes=@n14.MaterialAttributes
   BaseColor=@n30
   Specular=LinearInterpolate(A=@n14.Specular, B=ScalarParameter() {defaultValue=-0.5, group=Foam and Pondscum, parameterName=Foam Specular}, Alpha=@n40)
   Roughness=LinearInterpolate(A=@n14.Roughness, B=ScalarParameter() {defaultValue=0.5, group=Foam and Pondscum, parameterName=Foam Roughness}, Alpha=@n40)
   Opacity=@n41
   Normal=LinearInterpolate(A=@n14.Normal, B=FunctionOutput(None=@n43) {outputName=Normal, sortPriority=4}, Alpha=@n40)
  ) {attributeSetTypes [
    #2
    #3
    #4
    #5
    #6
   ]}
  n45 = SetMaterialAttributes!(MaterialAttributes=@n14.MaterialAttributes, BaseColor=@n30, Opacity=@n41) {attributeSetTypes [#2, #5]}
  n46 = QualitySwitch(Default=@n44, Low=@n45, High=@n44, Medium=@n45)
  FunctionOutput(None=ShadingPathSwitch(Default=@n46, Deferred=@n46, Forward=@n45, Mobile=@n45)) {outputName=Attributes}
  TextureSample(UVs=@n21, Bias=ScalarParameter() {group=Foam and Pondscum, parameterName=Foam Normal Blur}) {mipValueMode=TMVM_MipBias, samplerSource=SSM_Wrap_WorldGroupSettings, samplerType=SAMPLERTYPE_Normal}
 ]}
