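# Material function for distant water that blends an incoming normal toward a far-wave normal and fades World Position Offset by the same far mask.
# Two world-XY samples of one tiling normal pan at different scales and directions, then are summed and flattened by an intensity parameter.
# The far mask multiplies the override and additional mask inputs.
# Low quality, forward and mobile paths output only the far normal, with flatness adjusted by a low-detail biome mask.
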MaterialFunction /Clyde/Materials/Functions/WaterTwoTexturePanner
graph {nodes [
  Constant3Vector() {constant=<rgba 0 0 1 0>}
  n1 = Divide(
   A=ComponentMask(None=WorldPosition() {worldPositionShaderOffset=WPT_ExcludeAllShaderOffsets}.XYZ) {g=true, r=true}
   B=ScalarParameter() {defaultValue=8000, group=Distance Blending, parameterName=FarWaveNormalScale}
  )
  n2 = MaterialFunctionCall(Speed=ScalarParameter() {defaultValue=1.6, group=Distance Blending, parameterName=FarWaveNormalSpeed}) {materialFunction=&TimeWithSpeedVariable}
  n3 = ScalarParameter() {defaultValue=1, group=Distance Blending, parameterName=FarMipBias}
  n4 = TextureSampleParameter2D!(
   UVs=Panner(
    Coordinate=@n1
    Time=@n2.Time
    Speed=VectorParameter() {group=Distance Blending, parameterName=Distance Normal Panner 1, defaultValue=<rgba 0.05 0.02 0 0>}.RGB
   )
   Bias=@n3
   Apply View MipBias
  ) {
   group=Distance Blending
   mipValueMode=TMVM_MipBias
   parameterName=FarWaveNormalHeight
   samplerSource=SSM_Wrap_WorldGroupSettings
   samplerType=SAMPLERTYPE_LinearColor
   texture=/Clyde/ExtraDependencies/T_Water_TilingNormal_With_Height_02_Softened
  }
  n5 = ComponentMask(None=@n4.RGB) {g=true, r=true}
  n6 = TextureSampleParameter2D(
   UVs=Panner(
    Coordinate=Multiply(A=@n1) {constB=1.618}
    Time=@n2.Time
    Speed=VectorParameter() {group=Distance Blending, parameterName=Distance Normal Panner 2, defaultValue=<rgba -0.01 -0.06 0 0>}.RGB
   )
   Bias=@n3
  ) {
   group=Distance Blending
   mipValueMode=TMVM_MipBias
   parameterName=FarWaveNormalHeight
   samplerSource=SSM_Wrap_WorldGroupSettings
   samplerType=SAMPLERTYPE_LinearColor
   texture=&T_Water_TilingNormal_With_Height_02_Softened
  }
  n7 = ComponentMask(None=@n6.RGB) {g=true, r=true}
  n8 = AppendVector(A=Add(A=ConstantBiasScale(None=@n5) {bias=-0.5, scale=2}, B=ConstantBiasScale(None=@n7) {bias=-0.5, scale=2}), B=Constant())
  n9 = OneMinus(None=ScalarParameter() {defaultValue=0.05, group=Distance Blending, parameterName=FarWavelNormalIntensity})
  n10 = Constant()
  n11 = FunctionInput() {bUsePreviewValueAsDefault=true, inputName=Additional Mask, inputType=FunctionInput_Scalar, sortPriority=3, previewValue=<1 0 0 1>}
  n12 = FunctionOutput(None=Multiply(
    A=FunctionInput(Preview=@n10) {bUsePreviewValueAsDefault=true, inputName=Far Mask Override, inputType=FunctionInput_Scalar, sortPriority=2}
    B=@n11
   )) {outputName=Far Mask, sortPriority=3}
  n13 = LinearInterpolate(
   A=FunctionInput() {bUsePreviewValueAsDefault=true, inputName=Normal, previewValue=<0 0 1 1>}
   B=MaterialFunctionCall(Normal=@n8, Flatness=@n9) {materialFunction=&FlattenNormal}.Result
   Alpha=@n12
  )
  n14 = FunctionInput() {bUsePreviewValueAsDefault=true, inputName=LowDetailBiome_Mask, inputType=FunctionInput_Scalar, sortPriority=8, previewValue=<1 0 0 1>}
  n15 = MaterialFunctionCall(
   Normal=@n8
   Flatness=LinearInterpolate(
    A=@n9
    B=ScalarParameter() {defaultValue=0.99, group=Distance Blending, parameterName=FarWavelNormalIntensity_Biome_LowDetail}
    Alpha=Saturate(None=@n14)
   )
  ) {materialFunction=&FlattenNormal}
  n16 = QualitySwitch(Default=@n13, Low=@n15.Result, High=@n13, Medium=@n13)
  FunctionOutput(None=ShadingPathSwitch(Default=@n16, Deferred=@n16, Forward=@n15.Result, Mobile=@n15.Result)) {outputName=Normal}
  FunctionOutput(None=Multiply(A=FunctionInput() {bUsePreviewValueAsDefault=true, inputName=WPO, sortPriority=1}, B=OneMinus(None=@n12))) {outputName=WPO, sortPriority=1}
  MaterialFunctionCall() {materialFunction=&/GetWaterTime}
 ]}
