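# Material function that layers jungle algae scum onto incoming water material attributes, behind the Use_JungleAlgea switch, passing the input through on mobile and Low quality.
# World-XY JungleWater textures form a large mask and a ripple-distorted detail pattern, combined through HeightLerp and reduced by erasure.
# That mask blends in the two-tone scum color, flattened normal, opacity and roughness.
# It also outputs algae and scattering masks; the depth-mask and extra-large texture branches are unconnected.
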MaterialFunction /FortWater/Materials/Functions/MF_JungleWater
graph {nodes [
  n1 = Constant()
  n2 = Saturate(None=Divide(
    A=FunctionInput(Preview=@n1) {bUsePreviewValueAsDefault=true, inputName=MaskfromDepth, inputType=FunctionInput_Scalar, sortPriority=4}
    B=ScalarParameter() {defaultValue=200, group=Jungle, parameterName=Jungle_AgleaDepth}
   ))
  Add(A=Multiply(A=OneMinus(None=@n2), B=ScalarParameter() {defaultValue=0.5, group=Jungle, parameterName=Jungle_AgleaDepth_intensity}))
  Constant()
  n3 = FunctionInput() {inputType=FunctionInput_MaterialAttributes}
  n4 = GetMaterialAttributes(None=@n3) {attributeGetTypes [#1]}
  n5 = NamedRerouteUsage()
  n6 = ComponentMask(None=WorldPosition().XYZ) {g=true, r=true}
  n7 = TextureSample(UVs=Divide(A=@n6, B=ScalarParameter() {defaultValue=5000, group=Jungle, parameterName=Jungle_Large_TextureSize_mask})) {samplerSource=SSM_Wrap_WorldGroupSettings, texture=/FortWater/Textures/JungleWater_M}
  n8 = Saturate(None=LinearInterpolate(
    A=ScalarParameter() {group=Jungle, parameterName=Jungle_Large_Contrast_low}
    B=ScalarParameter() {defaultValue=1, group=Jungle, parameterName=Jungle_Large_Contrast_high}
    Alpha=@n7.G
   ))
  n9 = FunctionInput() {bUsePreviewValueAsDefault=true, inputName=SimWaterRipple, sortPriority=20}
  n10 = Add(
   A=Divide(A=@n6, B=ScalarParameter() {defaultValue=426.3825, group=Jungle, parameterName=Jungle_Detail_TextureSize})
   B=Multiply(A=ScalarParameter() {defaultValue=0.2, group=Jungle, parameterName=Jungle_Distort_Texture}, B=ComponentMask(None=@n9) {g=true, r=true})
  )
  n11 = Saturate(None=LinearInterpolate(
    A=ScalarParameter() {group=Jungle, parameterName=Jungle_Detail_Contrast_low}
    B=ScalarParameter() {defaultValue=1, group=Jungle, parameterName=Jungle_Detail_Contrast_high}
    Alpha=TextureSample(UVs=@n10) {samplerSource=SSM_Wrap_WorldGroupSettings, texture=&JungleWater_M}.R
   ))
  n12 = MaterialFunctionCall(Transition Phase=@n8, Height Texture=@n11, Contrast=ScalarParameter() {group=Jungle, parameterName=Jungle_Contrast_HeightLerp}) {materialFunction=&HeightLerp}
  n13 = Saturate(None=Subtract(A=@n12.Alpha, B=NamedRerouteUsage()))
  n14 = Saturate(None=@n13)
  n15 = Saturate(None=Multiply(A=@n13, B=ScalarParameter() {defaultValue=50, group=Jungle, parameterName=Jungle_ColorMask_Mult}))
  n16 = Saturate(None=LinearInterpolate(
    A=ScalarParameter() {group=Jungle, parameterName=Jungle_AlgeaColorContrast_low}
    B=ScalarParameter() {defaultValue=3, group=Jungle, parameterName=Jungle_AlgeaColorContrast_high}
    Alpha=@n12.Lerp Alpha No Contrast
   ))
  n17 = LinearInterpolate(
   A=VectorParameter() {group=Jungle, parameterName=Jungle_Color_Algea, defaultValue=<rgba 0.1073 0.1615 0.0209 1>}.RGB
   B=VectorParameter() {group=Jungle, parameterName=Jungle_Color_Algea_2, defaultValue=<rgba 0.1073 0.1615 0.0209 1>}.RGB
   Alpha=@n16
  )
  n18 = StaticSwitchParameter(
   True=SetMaterialAttributes!(
    MaterialAttributes=BlendMaterialAttributes(A=@n4.MaterialAttributes, B=@n5, Alpha=@n14) {vertexAttributeBlendType=UseA}
    BaseColor=LinearInterpolate(A=@n4.BaseColor, B=@n17, Alpha=@n15)
   ) {attributeSetTypes [#1]}
   False=@n3
  ) {parameterName=Use_JungleAlgea}
  FunctionOutput(None=ShadingPathSwitch(Default=QualitySwitch(Default=@n18, Low=@n3, High=@n18, Medium=@n18, Epic=@n18), Mobile=@n3) {desc=Migrated from FeatureLevelSwitch})
  n19 = ScalarParameter() {defaultValue=0.5, group=Jungle, parameterName=Jungle_ScatterMask_Mult}
  FunctionOutput(None=Saturate(None=Multiply(A=@n13, B=@n19))) {outputName=Mask, sortPriority=4}
  n20 = LinearInterpolate(
   A=ScalarParameter() {group=Jungle, parameterName=Jungle_Scatter_Contrast_low}
   B=ScalarParameter() {defaultValue=2, group=Jungle, parameterName=Jungle_Scatter_Contrast_high}
   Alpha=@n7.G
  )
  n21 = Add(
   A=Multiply(A=NamedRerouteUsage(), B=ScalarParameter() {defaultValue=0.5, group=Jungle, parameterName=Jungle_MaxErasure})
   B=OneMinus(None=Saturate(None=@n20))
  )
  FunctionOutput(None=OneMinus(None=Saturate(None=@n21))) {outputName=Mask_Scattering, sortPriority=4}
  Multiply(A=@n11, B=@n8)
  n22 = ScalarParameter() {defaultValue=10000, group=Jungle, parameterName=Jungle_ExtraLarge_TextureSize_mask}
  Multiply(A=@n7.G, B=TextureSample(UVs=Divide(A=@n6, B=@n22)) {samplerSource=SSM_Wrap_WorldGroupSettings, texture=&JungleWater_M}.G)
  n23 = FunctionInput(Preview=@n1) {bUsePreviewValueAsDefault=true, inputName=ErasureMask, inputType=FunctionInput_Scalar, sortPriority=10}
  NamedRerouteDeclaration(None=@n23) {name=ErasureMask, nodeColor=<rgba 0.875 1 0 1>}
  n24 = SetMaterialAttributes!(
   MaterialAttributes
   BaseColor=@n17
   Normal=MaterialFunctionCall(
    Normal=TextureSample(UVs=@n10) {samplerSource=SSM_Wrap_WorldGroupSettings, samplerType=SAMPLERTYPE_Normal, texture=/FortWater/Textures/JungleWater_N}.RGB
    Flatness=ScalarParameter() {defaultValue=0.5, group=Jungle, parameterName=Jungle_AlgeaNormal}
   ) {materialFunction=&FlattenNormal}.Result
   Opacity=LinearInterpolate(
    A=ScalarParameter() {defaultValue=1, group=Jungle, parameterName=Jungle_Algea_Opacity2}
    B=ScalarParameter() {defaultValue=1, group=Jungle, parameterName=Jungle_Algea_Opacity}
    Alpha=@n16
   )
   Roughness=ScalarParameter() {defaultValue=0.5, group=Jungle, parameterName=Jungle_Algea_Roughness}
   Specular=ScalarParameter() {defaultValue=0.5, group=Jungle, parameterName=Jungle_Algea_Specular}
  ) {attributeSetTypes [
    #1
    #2
    #3
    #4
    #5
   ]}
  NamedRerouteDeclaration(None=@n24) {name=ScumMaterial, nodeColor=<rgba 0.9453 0 1 1>}
  Reroute(None=@n23)
  Reroute(None=@n24)
 ]}
