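# M_Water_Ocean_02_NewModel is a masked, two-sided Single Layer Water ocean surface material.
# Gerstner waves, attenuated by depth, feed panning normals and a fluid-sim texture, with height-map and camera-bias vertex offsets.
# Foam, swamp scum and jungle-water layers follow, while scattering and absorption blend by noise, biome and macro variation.
# Static switches add clipped holes, far-water blending and location-based masks, with simpler paths for lower quality and mobile.
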Material /Clyde/Materials/WaterSurface/M_Water_Ocean_02_NewModel
graph {
 properties {
  allowTranslucentCustomDepthWrites=true
  bTangentSpaceNormal=false
  bUseMaterialAttributes=true
  bUsedWithInstancedStaticMeshes=true
  bUsedWithNanite=true
  bUsedWithSplineMeshes=true
  bUsedWithWater=true
  blendMode=BLEND_Masked
  pixelDepthOffsetMode=PDOM_Legacy
  refractionMethod=RM_PixelNormalOffset
  shadingModel=MSM_SingleLayerWater
  twoSided=true
  usedShadingModels=SingleLayerWater
 }
 nodes [
  n1 = NamedRerouteUsage()
  n2 = NamedRerouteUsage()
  n3 = NamedRerouteUsage()
  MaterialExpressionSaturate_38 = Saturate(None=Multiply(A=NamedRerouteUsage(), B=NamedRerouteUsage()))
  n4 = StaticSwitchParameter(
   True=ScalarParameter() {parameterName=WaterBodyIndexOverride}
   False=StaticSwitchParameter(
    True=ScalarParameter() {parameterName=WaterBodyIndex}
    False=Custom() {code="GetWaterWaveParamIndex(Parameters)", description=WaterBodyIndex From Tile Instance, outputType=CMOT_Float1, inputs []}
   ) {parameterName=UseScalarWaterBodyIndex}
  ) {parameterName=OverrideWaterBodyIndex}
  MaterialExpressionMaterialFunctionCall_69 = MaterialFunctionCall!(Z Offset Only, Position, WaterBodyIndex=@n4, Normals use VS Interpolation, All Normals VS) {materialFunction=/Clyde/Materials/Functions/Fort_ComputeGerstnerWaves}
  n5 = Custom!(ClipMask=@MaterialExpressionSaturate_38, NormalPassThru=@MaterialExpressionMaterialFunctionCall_69.Normal) {
   code="##if !MATERIALBLENDING_MASKED\r\n##error \"UseOceanHole is only supported on Masked material blending modes\"!\r\n##endif\r\nif (ClipMask == 0)\r\n  clip(-1);\r\nreturn NormalPassThru;"
   description=Clip Hole Mask
   inputs [
    {inputName=ClipMask, input {expression=&self:MaterialExpressionSaturate_38}}
    {
     inputName=NormalPassThru
     input {expression=&self:MaterialExpressionMaterialFunctionCall_69}
    }
   ]
  }
  n6 = MaterialFunctionCall!(Texture, Position) {materialFunction=/FortWater/Materials/Functions/WaterBodyData}
  n7 = NamedRerouteUsage()
  n8 = MaterialFunctionCall!(
   Normal=StaticSwitchParameter(True=@n5, False=@MaterialExpressionMaterialFunctionCall_69.Normal) {group=Water Hole, parameterName=UseOceanHole}
   WPO=@MaterialExpressionMaterialFunctionCall_69.WPO
   Mask=Multiply(
    A=MaterialFunctionCall!(WaterDepth=@n6.Water Depth, Water Body Index=@n4) {materialFunction=/Water/Materials/Functions/ComputeWavesAttenuationFactor}.WaveAttenuationFactor
    B=OneMinus(None=@n7)
   )
  ) {materialFunction=/FortWater/Materials/Functions/MaskOutWaves}
  n9 = Constant3Vector() {constant=<rgba 0 0 1 0>}
  n10 = NamedRerouteUsage()
  n11 = NamedRerouteUsage()
  n12 = MaterialFunctionCall!(
   Normal=StaticSwitch(True=LinearInterpolate(A=@n8.Normal, B=@n9, Alpha=@n10), False=@n8.Normal, Value=@n11)
   WPO=@n8.WPO
   Far Mask Override=@MaterialExpressionMaterialFunctionCall_69.Distance Mask
   Additional Mask=StaticSwitch(True=OneMinus(None=@n10), False=Constant() {r=1}, Value=@n11)
   LowDetailBiome_Mask=Constant() {r=1}
  ) {materialFunction=/Clyde/Materials/Functions/WaterTwoTexturePanner}
  n13 = ScalarParameter() {defaultValue=35000, parameterName=ZOffsetOffset}
  n14 = WorldPosition() {worldPositionShaderOffset=WPT_ExcludeAllShaderOffsets}
  n15 = ComponentMask(None=WorldPosition().XYZ) {b=true}
  n16 = StaticSwitchParameter(True=Multiply(A=Subtract(A=@n6.Water Z Height, B=@n15), B=Constant3Vector() {constant=<rgba 0 0 1 0>}), False=Constant()) {defaultValue=true, parameterName=UseHeightMappingVertexShader}
  n17 = Normalize(VectorInput=Subtract(A=CameraPositionWS(), B=Add(A=@n16, B=@n14.XYZ)))
  n18 = ScalarParameter() {defaultValue=256, parameterName=DistanceCameraOffset}
  n19 = Add(
   A=@n16
   B=Multiply(
    A=MaterialFunctionCall(Fade Length=@n13, Fade Offset=@n13) {materialFunction=&CameraDepthFade}.Result
    B=Multiply(A=@n17, B=@n18)
   )
  )
  n20 = MaterialFunctionCall!(
   Flowmap Velocity
   Previous Foam
   Normal=@n12.Normal
   WPO=Add(A=@n12.WPO, B=StaticSwitchParameter(True=@n19, False=@n16) {parameterName=UseCameraBiasVertexShader})
  ) {materialFunction=/Clyde/Materials/Functions/WaterSimulationTexture}
  n21 = NamedRerouteDeclaration(None=@n20.Sim Foam Erasure) {name=SimFoamErasure, nodeColor=<rgba 0.2891 1 0 1>}
  n22 = NamedRerouteUsage()
  n23 = NamedRerouteUsage()
  n24 = StaticSwitch(True=LinearInterpolate(B=@MaterialExpressionMaterialFunctionCall_69.Distance Mask, Alpha=@n10) {constA=1}, False=@n12.Far Mask, Value=@n11)
  n25 = StaticComponentMaskParameter(None=NamedRerouteUsage()) {defaultR=true, group=MacroColorVariation, parameterName=MacroMaskChannelForSpecular}
  n26 = ScalarParameter() {defaultValue=1, group=MacroColorVariation, parameterName=MacroMaskSpecular_Power}
  n27 = MaterialFunctionCall!(InSpecular=ScalarParameter() {defaultValue=0.225, group=Water Shading Model, parameterName=Specular}, WaterDepth=@n6.Water Depth) {materialFunction=/FortWater/Materials/Functions/MF_WaterBodySpecular}
  n28 = StaticSwitchParameter(
   True=LinearInterpolate(
    A=@n27.Result
    B=ScalarParameter() {defaultValue=0.05, group=MacroColorVariation, parameterName=Specular_MacroMask}
    Alpha=Power(Base=@n25, Exp=@n26)
   )
   False=@n27.Result
  ) {desc=Bad implemenation causes seams in water body transition, group=MacroColorVariation, parameterName=UseMacroColorMaskToBlendSpecular}
  n29 = NamedRerouteUsage()
  n30 = NamedRerouteUsage()
  n31 = Constant() {desc=Distance from water z center to far water height, r=56}
  n32 = Multiply(A=@n31, B=Constant3Vector() {constant=<rgba 0 0 -1 0>}) {constB=-1}
  n33 = NamedRerouteUsage()
  n34 = SetMaterialAttributes!(
   MaterialAttributes
   Base Color=VectorParameter() {parameterName=BaseColor, defaultValue=<rgba 0 0.1 0.1 0.1>}.RGB
   Specular=@n28
   Roughness=ScalarParameter() {defaultValue=0.05, desc=below 0.5 DLSS has sparkly bbloom highlights, group=Water Shading Model, parameterName=Roughness}
   Opacity=ScalarParameter() {group=Water Shading Model, parameterName=Surface Opacity}
   Normal=StaticSwitch(True=LinearInterpolate(A=@n20.Normal, B=@n12.Normal, Alpha=@n29), False=@n20.Normal, Value=NamedRerouteUsage())
   World Position Offset=StaticSwitch(True=LinearInterpolate(A=@n30, B=@n32, Alpha=@n33), False=@n30, Value=NamedRerouteUsage())
  ) {attributeSetTypes [
    #1
    #2
    #3
    #4
    #5
    #6
   ]}
  n35 = MaterialFunctionCall!(
   Attributes=@n34
   Sim Erasure=@n21
   Mask=@n22
   Far Mask=@n24
   Depth=@n23
   WaveHeight=ComponentMask(None=@MaterialExpressionMaterialFunctionCall_69.WPO) {b=true}
  ) {materialFunction=/FortWater/Materials/Functions/OceanShoreFoam_2}
  n36 = MaterialFunctionCall!(Attributes=@n34, Sim Erasure=@n21, Mask=@n22, Far Mask=@n24, Depth=@n23) {materialFunction=/Clyde/Materials/Functions/OceanShoreFoam}
  n37 = StaticSwitchParameter(True=StaticSwitchParameter(True=@n35.Attributes, False=@n36.Attributes) {defaultValue=true, parameterName=UseNewOceanFoam2}, False=@n34) {defaultValue=true, parameterName=AllowFoam}
  n38 = MaterialFunctionCall!(
   Attributes=StaticSwitchParameter(
    True=MaterialFunctionCall!(Attributes=@n37, Sim Erasure=@n1, Alpha Mask=@n2, Depth=@n3) {materialFunction=/FortWater/Materials/Functions/PondScumLayer}.Attributes
    False=@n37
   ) {parameterName=Allow Swamp}
   Depth=NamedRerouteUsage()
  ) {materialFunction=/FortWater/Materials/Functions/TwoSidedWaterCorrections}
  n39 = MaterialFunctionCall!(
   In=MaterialFunctionCall!(Attributes=@n38.Result) {desc=Simple Spec and Roughness On Mobile, materialFunction=/FortWater/Materials/Functions/WaterQualityOptimizations}.Result
   MaskfromDepth=NamedRerouteUsage()
   ErasureMask=NamedRerouteUsage()
   SimWaterRipple=NamedRerouteUsage()
  ) {materialFunction=/FortWater/Materials/Functions/MF_JungleWater}
  n40 = ScalarParameter() {defaultValue=-180, parameterName=Water opacity Offset}
  n41 = ScalarParameter() {parameterName=Water opacity Offset Far}
  n42 = Saturate(None=Divide(A=PixelDepth().R, B=ScalarParameter() {defaultValue=25000, parameterName=Water Opacity Offset Fade Distance}) {constB=15000})
  n43 = Subtract(
   A=MaterialFunctionCall() {materialFunction=&WaterBodyData}.Water Depth
   B=StaticSwitchParameter(True=LinearInterpolate(A=@n40, B=@n41, Alpha=@n42) {constB=0}, False=@n40) {defaultValue=true, parameterName=Use DIstant Opacity Bias}
  )
  n44 = Saturate(None=Divide(
    A=NamedRerouteDeclaration(None=@n43) {name=Water Depth Distance Biased, nodeColor=<rgba 0.2188 0 1 1>}
    B=ScalarParameter() {defaultValue=48, parameterName=Water Opacity Gradient}
   ))
  n45 = Divide(A=Subtract(A=SceneDepthWithoutWater!(UVs).R, B=PixelDepth().R), B=ScalarParameter() {defaultValue=32, parameterName=Depth Edge Gradient})
  n46 = MaterialFunctionCall!(Attributes=@n39.Result, Opacity Mask=Multiply(A=@n44, B=Saturate(None=@n45))) {desc=Use DepthFade on Opacity for Foam/Swamp, materialFunction=/FortWater/Materials/Functions/WaterOpacityMultiply}
  n47 = MaterialFunctionCall!(Attributes=@n46.Result) {materialFunction=/FortWater/Materials/Functions/WaterPhysicsForce}
  Composite!(InMA=@n47.Result) {desc=to match normal intensity of Epic, subgraphName=NormalizeNormal}
  MaterialExpressionReroute_9 = Reroute(None=@n47.Result)
  n48 = SetMaterialAttributes!(
   MaterialAttributes=@MaterialExpressionReroute_9.InMA
   Normal=Normalize(VectorInput=GetMaterialAttributes(None=@MaterialExpressionReroute_9.InMA) {attributeGetTypes [#5]}.Normal)
  ) {attributeSetTypes [#5]}
  MaterialExpressionReroute_12 = Reroute(None=@n48)
  n49 = MaterialFunctionCall!(Attributes=MaterialFunctionCall!(Attributes=QualitySwitch(Default=@n46.Result, High=@MaterialExpressionReroute_12.OutMA, Medium=@MaterialExpressionReroute_12.OutMA, Epic=@n47.Result)) {materialFunction=/Water/Materials/Functions/ClipDynamicBounds}.Attributes) {materialFunction=/FortWater/Materials/Functions/ClipWaterHitProxyFromDepth}
  n50 = NamedRerouteUsage()
  MaterialFunctionCall!(InMA=StaticSwitchParameter(
    True=SetMaterialAttributes!(MaterialAttributes=@n49.Result, Opacity Mask=@n50) {attributeSetTypes [#7]}
    False=@n49.Result
   ) {parameterName=Force Opacity Mask Hole}) {materialFunction=/Game/Tools/NewWaterfallGenerator/Material/WaterfallClipBound}
  n51 = WorldPosition()
  n52 = CollectionParameter() {collection=/Clyde/Materials/MPC/MPC_Water, parameterName=SimLocation}
  n53 = Abs(None=Subtract(A=ComponentMask(None=@n51.XYZ) {g=true, r=true}, B=ComponentMask(None=@n52) {g=true, r=true}))
  n54 = DoubleVectorParameter() {parameterName=WaterArea}
  n55 = CollectionParameter() {collection=&MPC_Water, parameterName=FluidSimSize}
  n56 = Multiply(A=AppendVector(A=@n55, B=@n55)) {constB=4}
  n57 = Subtract(A=AppendVector(A=@n54.Z, B=@n54.Z) {desc=Get the extent vector from the texture}, B=@n56) {constA=100, constB=100}
  n58 = Saturate(None=Divide(A=Subtract(A=@n53, B=@n57), B=@n56))
  n59 = Max(A=ComponentMask(None=@n58) {r=true}, B=ComponentMask(None=@n58) {g=true})
  n60 = Constant()
  NamedRerouteDeclaration(None=StaticSwitch(
    True=StaticSwitchParameter(True=@n59, False=@n60) {defaultValue=true, parameterName=UseOceanFarWaterBlending}
    False=@n60
    Value=NamedRerouteUsage()
   )) {name=MASK_FarWaterTransition, nodeColor=<rgba 0.5 0 1 1>}
  n61 = QualitySwitch(Default=StaticBool(), Epic=StaticBool() {value=true})
  n62 = StaticBool()
  NamedRerouteDeclaration(None=StaticSwitch(
    True=ShadingPathSwitch(Default=@n61, Deferred=@n61, Forward=@n62, Mobile=@n62)
    False=@n62
    Value=StaticBoolParameter() {desc=Should ocean materials smoothly blend into the far water mesh at its edges?, parameterName=Use Far Water Blending}
   )) {name=FarWaterBlendingEnabled, nodeColor=<rgba 1 0.4922 0 1>}
  n63 = VectorParameter() {group=Water Shading Model, parameterName=Scattering, defaultValue=<rgba 0.1268 0.1336 0.15 0.5>}
  n64 = NamedRerouteDeclaration(None=Multiply(A=@n63.RGB, B=Divide(A=@n63.A) {constB=100})) {name=Scattering_Original, nodeColor=<rgba 0 1 0.0391 1>}
  n65 = VectorParameter() {group=Water Shading Model, parameterName=Scattering_Noise, defaultValue=<rgba 0.1268 0.1336 0.15 0.5>}
  n66 = Divide(
   A=ComponentMask(None=WorldPosition().XYZ) {g=true, r=true}
   B=ScalarParameter() {defaultValue=130000, group=Water Shading Model, parameterName=OceanColorNoise_Size}
  )
  n67 = Saturate(None=Add(A=ScalarParameter() {group=Water Shading Model, parameterName=BlendOceanNoiseByUVs_GradientBias}, B=@n66))
  n68 = AppendVector(A=@n67, B=OneMinus(None=@n67))
  n69 = TextureSample(UVs=@n66) {samplerSource=SSM_Wrap_WorldGroupSettings, samplerType=SAMPLERTYPE_Masks, texture=/FortWater/Textures/OceanNoise__Mask}
  n70 = StaticSwitchParameter(
   True=Multiply(
    A=StaticComponentMaskParameter(None=@n68) {defaultG=true, defaultR=true, group=Water Shading Model, parameterName=BlendOceanNoiseByUVs_Direction}
    B=@n69.R
   )
   False=@n69.R
  ) {group=Water Shading Model, parameterName=BlendOceanNoiseByUVs}
  n71 = StaticSwitchParameter(True=NamedRerouteUsage(), False=@n70) {parameterName=Use_JungleAlgea}
  n72 = FeatureLevelSwitch(Default=QualitySwitch(Default=@n71, Low=@n70, High=@n71, Medium=@n71, Epic=@n71), ES3_1=@n70)
  n73 = LinearInterpolate(A=@n64, B=Multiply(A=@n65.RGB, B=Divide(A=@n65.A) {constB=100}), Alpha=@n72)
  MaterialExpressionReroute_0 = Reroute(None=StaticSwitchParameter(True=FeatureLevelSwitch(Default=@n73, ES3_1=@n64), False=@n64) {parameterName=UseOceanColorNoise})
  n74 = Add(A=NamedRerouteUsage(), B=@MaterialExpressionReroute_0)
  MaterialExpressionSaturate_37 = Saturate(None=Multiply(A=NamedRerouteUsage(), B=NamedRerouteUsage()))
  n75 = Custom!(ScatteringPassThru=@MaterialExpressionReroute_0, ClipMask=@MaterialExpressionSaturate_37) {
   code="##if !MATERIALBLENDING_MASKED\r\n##error \"UseOceanHole is only supported on Masked material blending modes\"!\r\n##endif\r\nif (ClipMask == 0)\r\n  clip(-1);\r\nreturn ScatteringPassThru;"
   description=Clip Hole Mask
   inputs [
    {inputName=ScatteringPassThru, input {expression=&self:MaterialExpressionReroute_0}}
    {inputName=ClipMask, input {expression=&self:MaterialExpressionSaturate_37}}
   ]
  }
  n76 = StaticSwitchParameter(True=@n75, False=@MaterialExpressionReroute_0) {group=Water Hole, parameterName=UseOceanHole}
  n77 = QualitySwitch(Default=@n74, Low=@n76, High=@n74, Medium=@n76)
  n78 = ShadingPathSwitch(Default=@n77, Deferred=@n77, Forward=@n76, Mobile=@n76)
  n79 = VectorParameter() {group=Water Shading Model, parameterName=BiomeScattering, defaultValue=<rgba 0.2301 0.2396 0.1547 1>}
  n80 = VectorParameter() {group=Water Shading Model, parameterName=Absorption, defaultValue=<rgba 1 13.35 20 100>}
  n81 = Multiply(A=@n80.RGB, B=@n80.RGB)
  n82 = NamedRerouteDeclaration(None=@n81) {name=Absorbtion_Original, nodeColor=<rgba 0 1 0.9297 1>}
  n83 = VectorParameter() {group=Water Shading Model, parameterName=Absorption_Noise, defaultValue=<rgba 1 13.35 20 100>}
  n84 = Multiply(A=@n83.RGB, B=@n83.RGB)
  n85 = StaticSwitchParameter(True=FeatureLevelSwitch(Default=LinearInterpolate(A=@n82, B=@n84, Alpha=@n72), ES3_1=@n82), False=@n82) {parameterName=UseOceanColorNoise}
  n86 = VectorParameter() {group=Water Shading Model, parameterName=BiomeAbsorption, defaultValue=<rgba 1 0.642 0.4152 100>}
  n87 = MaterialFunctionCall!(
   Scattering A=@n78
   Scattering B=Multiply(A=@n79.RGB, B=Divide(A=@n79.A) {constB=100})
   Absorption A=@n85
   Absorption B=Multiply(A=@n86.RGB, B=@n86.RGB)
   Alpha=MaterialFunctionCall(In=NamedRerouteUsage(), Power=ScalarParameter() {defaultValue=2, parameterName=BiomePower}) {materialFunction=&SCurve}.Result
  ) {materialFunction=/FortWater/Materials/Functions/BlendWaterShadingValues}
  n88 = MaterialFunctionCall!(
   InScattering=StaticSwitchParameter(True=@n87.Scattering, False=@n78) {parameterName=Allow Swamp}
   InAbsorption=StaticSwitchParameter(True=@n87.Absorption, False=@n85) {parameterName=Allow Swamp}
   InTransitionMask
   InOceanExtensionMask
   InNoise=StaticSwitchParameter(True=NamedRerouteUsage(), False=Constant() {r=1}) {parameterName=UseOceanColorNoise}
  ) {materialFunction=/FortWater/Materials/Functions/MF_WaterMacroColorVariation}
  n89 = MaterialFunctionCall(Float3=@n88.Mask) {materialFunction=&BreakOutFloat3Components}
  NamedRerouteDeclaration(None=@n89.B) {name=MacroMask_Mirrored_Negative, nodeColor=<rgba 0 1 0.2734 1>}
  n90 = ScalarParameter() {defaultValue=150000, parameterName=BoxBlend_Min}
  n91 = ScalarParameter() {defaultValue=200000, parameterName=BoxBlend_Max}
  n92 = Abs(None=WorldPosition().XY)
  n93 = MaterialFunctionCall(Float2=OneMinus(None=SmoothStep(Min=@n90, Max=@n91, Value=@n92) {constMax=0.5, constMin=0.4})) {materialFunction=&BreakOutFloat2Components}
  NamedRerouteDeclaration(None=Saturate(None=Multiply(A=@n93.R, B=@n93.R))) {name=BoxBlendMask, nodeColor=<rgba 1 0.8203 0 1>}
  NamedRerouteDeclaration(None=@n88.Mask) {name=MacroMaskPacked, nodeColor=<rgba 0.6172 1 0 1>}
  NamedRerouteDeclaration(None=@n88.Absorption) {name=FinalAbsorption, nodeColor=<rgba 0.5469 1 0 1>}
  NamedRerouteDeclaration(None=@n88.Scattering) {name=FinalScatter, nodeColor=<rgba 1 0.9141 0 1>}
  n94 = NamedRerouteUsage()
  n95 = Add(A=@n8.WPO, B=@n31) {constB=56}
  n96 = Add(A=@n31, B=@n31)
  n97 = Saturate(None=ComponentMask(None=Divide(A=@n95, B=@n96) {constB=128}) {b=true})
  n98 = LinearInterpolate(A=@n94, Alpha=Subtract(A=@n94, B=OneMinus(None=@n97)))
  NamedRerouteDeclaration(None=StaticSwitch(True=Saturate(None=@n98), False=Constant(), Value=NamedRerouteUsage())) {name=MASK_FarWaterTransitionDistorted, nodeColor=<rgba 0 1 0.4375 1>}
  NamedRerouteDeclaration(None=@n39.Mask) {name=JungleAlgea_Mask, nodeColor=<rgba 1 0.1875 0 1>}
  n99 = ComponentMask(None=WorldPosition().XYZ) {g=true, r=true}
  n100 = AppendVector(
   A=ScalarParameter() {group=Water Hole, parameterName=WaterHoleMask_XOffset}
   B=ScalarParameter() {group=Water Hole, parameterName=WaterHoleMask_YOffset}
  )
  n101 = TextureSampleParameter2D!(
   UVs=Divide(A=Add(A=@n99, B=@n100), B=ScalarParameter() {defaultValue=5000, group=Water Hole, parameterName=WaterHoleMaskSize})
   Level
   Apply View MipBias
  ) {
   constMipValue=0
   group=Water Hole
   mipValueMode=TMVM_MipLevel
   parameterName=WaterHoleMask
   samplerSource=SSM_Clamp_WorldGroupSettings
   samplerType=SAMPLERTYPE_LinearColor
   texture=/FortWater/Textures/EmptyMask
  }
  n102 = Ceil(None=@n101.R)
  n103 = WorldPosition()
  n104 = VectorParameter() {group=Water Hole, parameterName=HolePosition}
  n105 = Subtract(
   A=Distance(A=ComponentMask(None=@n103.XYZ) {g=true, r=true}, B=ComponentMask(None=@n104.RGB) {g=true, r=true})
   B=ScalarParameter() {defaultValue=2048, group=Water Hole, parameterName=HoleRadius}
  )
  NamedRerouteDeclaration(None=StaticSwitchParameter(
    True=StaticSwitchParameter(True=@n102, False=@n105) {defaultValue=true, group=Water Hole, parameterName=UseMaskTextureForHole}
    False=Constant() {r=1}
   ) {group=Water Hole, parameterName=UseOceanHole}) {name=MASK_Hole, nodeColor=<rgba 1 0.7734 0 1>}
  n106 = ComponentMask(None=WorldPosition().XYZ) {g=true, r=true}
  n107 = CollectionParameter() {collection=&MPC_Water, parameterName=SwampColor}
  n108 = ComponentMask(None=@n107) {g=true, r=true}
  n109 = ComponentMask(None=@n107) {a=true}
  n110 = Multiply(
   A=ScalarParameter() {defaultValue=62000, group=Biomes, parameterName=Swamp Pondscum Radius}
   B=Saturate(None=Ceil(None=@n109)) {desc="Cancel swamp if radius <= 0"}
  )
  n111 = CollectionParameter() {collection=&MPC_Water, parameterName=SwampMaskHardness}
  NamedRerouteDeclaration(None=StaticSwitchParameter(True=SphereMask(A=@n106, B=@n108, Radius=@n110, Hardness=@n111) {desc=Swamp Pondscum Mask}, False=Constant()) {defaultValue=true, parameterName=Use Swamp Biome}) {name=MASK_SwampBiome, nodeColor=<rgba 0 1 0.9531 1>}
  NamedRerouteDeclaration(None=MaterialFunctionCall(In=@n101.R, Contrast=Constant() {r=0.5}) {materialFunction=&CheapContrast}.Result) {name=MASK_CustomTexture, nodeColor=<rgba 1 0 0.5625 1>}
  n112 = NamedRerouteUsage()
  n113 = VectorParameter() {parameterName=Cave Loc, defaultValue=<rgba -24616.0176 124808 3072 0>}
  n114 = SphereMask(
   A=ComponentMask(None=WorldPosition().XYZ) {g=true, r=true}
   B=ComponentMask(None=@n113.RGB) {g=true, r=true}
   Radius=@n113.B
   Hardness=Constant() {r=0.5}
  ) {desc=Swamp Pondscum Mask, hardnessPercent=50}
  n115 = StaticSwitchParameter(True=SphereMask(A=@n106, B=@n108, Radius=@n109, Hardness=@n111) {desc=Swamp Scattering Mask}, False=Constant()) {defaultValue=true, parameterName=Use Swamp Biome}
  NamedRerouteDeclaration(None=StaticSwitchParameter(True=OneMinus(None=@n112), False=StaticSwitchParameter(True=@n114, False=@n115) {parameterName=Cave River Transition}) {parameterName=UseSunkenPOIMaskHack}) {name=MASK_Cave/Swamp/Custom, nodeColor=<rgba 0.8281 0 1 1>}
  n116 = NamedRerouteUsage()
  NamedRerouteDeclaration(None=StaticSwitchParameter(True=Multiply(A=@n20.WPO, B=@n116), False=@n20.WPO) {parameterName=UseBoxBlendFlattenWPO}) {name=Water Sim WPO, nodeColor=<rgba 0 1 0.3203 1>}
  NamedRerouteDeclaration(None=StaticSwitchParameter(
    True=StaticSwitchParameter(True=@n35.Scatter Mask, False=@n36.Scatter Mask) {defaultValue=true, parameterName=UseNewOceanFoam2}
    False=Constant()
   ) {defaultValue=true, parameterName=AllowFoam}) {name=Ocean Foam Mask, nodeColor=<rgba 0.2656 0 1 1>}
  n117 = NamedRerouteUsage()
  n118 = ComponentMask(None=CollectionParameter() {collection=&MPC_Water, parameterName=YachtLocation}) {b=true, g=true, r=true}
  n119 = SphereMask(
   A=WorldPosition().XYZ
   B=StaticSwitchParameter(True=VectorParameter() {group=WPO SphereMaskHack, parameterName=MaskCenter, defaultValue=<rgba 79296 125129.0078 -3832 0>}.RGB, False=@n118) {group=WPO SphereMaskHack, parameterName=UseMaskCenterForWPOHack}
   Radius=ScalarParameter() {defaultValue=7500, group=WPO SphereMaskHack, parameterName=MaskRadius}
   Hardness=ScalarParameter() {defaultValue=0.2, group=WPO SphereMaskHack, parameterName=MaskHardness}
  )
  n120 = OneMinus(None=@n119)
  n121 = NamedRerouteUsage()
  n122 = Saturate(None=Distance(A=@n118, B=Constant()))
  n123 = OneMinus(None=Min(A=@n119, B=Ceil(None=@n122)))
  n124 = StaticSwitchParameter(
   True=StaticSwitchParameter(True=Multiply(A=@n120, B=@n121), False=Multiply(A=@n123, B=@n121)) {group=WPO SphereMaskHack, parameterName=UseMaskCenterForWPOHack}
   False=@n121
  ) {group=WPO SphereMaskHack, parameterName=UseWaveWPOSphereMaskHack}
  NamedRerouteDeclaration(None=StaticSwitchParameter(True=Multiply(A=@n117, B=@n124), False=@n124) {parameterName=UseSunkenPOIMaskHack}) {name=MASK_Yacht, nodeColor=<rgba 0.0755 1 0.1296 1>}
  n125 = NamedRerouteUsage()
  n126 = NamedRerouteUsage()
  n127 = WorldPosition()
  n128 = VectorParameter() {group=Biomes, parameterName=Foam Mask Location Radius, defaultValue=<rgba 81160 -98924 10000 1>}
  n129 = OneMinus(None=SphereMask(
    A=ComponentMask(None=@n127.XYZ) {g=true, r=true}
    B=ComponentMask(None=@n128.RGB) {g=true, r=true}
    Radius=@n128.B
    Hardness=ScalarParameter() {defaultValue=0.85, group=Biomes, parameterName=Foam Mask Hardness}
   ) {desc=Custom Foam Mask})
  n130 = Multiply(
   A=OneMinus(None=NamedRerouteUsage())
   B=StaticSwitchParameter(True=Multiply(A=@n126, B=@n129), False=@n129) {parameterName=UseSunkenPOIMaskHack}
  )
  n131 = OneMinus(None=NamedRerouteUsage())
  n132 = StaticSwitchParameter(True=StaticSwitch(True=Multiply(A=@n130, B=@n131), False=@n130, Value=NamedRerouteUsage()), False=Constant() {r=1}) {defaultValue=true, parameterName=UseCustomFoamMask}
  NamedRerouteDeclaration(None=StaticSwitchParameter(True=Multiply(A=@n125, B=@n132), False=@n132) {parameterName=UseBoxBlendFlattenColor}) {name=MASK_Custom Foam, nodeColor=<rgba 0.3359 1 0 1>}
  NamedRerouteDeclaration(None=@n6.Water Depth) {name=Base Water Depth, nodeColor=<rgba 0 1 0.2969 1>}
  NamedRerouteDeclaration(None=@n39.Mask_Scattering) {name=JungleAlgea_Scatter_Mask, nodeColor=<rgba 0 0.8828 1 1>}
  NamedRerouteDeclaration(None=@n20.Normal) {name=SimWaterNormal, nodeColor=<rgba 0 0.3437 1 1>}
  NamedRerouteDeclaration(None=@n70) {name=OceanNoise, nodeColor=<rgba 0.1953 0 1 1>}
  NamedRerouteDeclaration(None=@n89.R) {name=MacroMask, nodeColor=<rgba 0 1 0.2734 1>}
  NamedRerouteDeclaration(None=@n89.G) {name=MacroMask_Mirrored_Positive, nodeColor=<rgba 0 1 0.2734 1>}
  NamedRerouteUsage()
  PinBase() {reroutePins [InMA {expression=&self:MaterialExpressionReroute_9}]}
  PinBase!(OutMA=@n48) {reroutePins [OutMA {expression=&self:MaterialExpressionReroute_12}]}
  QualitySwitch(Default=StaticBool(), Low=StaticBool() {value=true})
  Reroute(None=@n81)
  Reroute(None=@n20.Normal)
  n133 = NamedRerouteUsage()
  n134 = Constant3Vector() {constant=<rgba 0 0 0 1>}
  n135 = NamedRerouteUsage()
  n136 = StaticSwitchParameter(True=LinearInterpolate(A=@n88.Scattering, B=@n134, Alpha=@n135), False=@n88.Scattering) {parameterName=Use_JungleAlgea}
  n137 = ComponentMask(None=CameraVectorWS()) {b=true}
  n138 = LinearInterpolate(A=Saturate(None=Ceil(None=@n137)), B=Multiply(A=@n38.NearMask, B=@n44), Alpha=Saturate(None=TwoSidedSign()))
  n139 = Multiply(A=@n138, B=FeatureLevelSwitch(Default=QualitySwitch(Default=@n136, High=@n136, Medium=@n136, Epic=@n136), ES3_1=@n88.Scattering))
  n140 = NamedRerouteUsage()
  n141 = Divide(B=NamedRerouteUsage())
  n142 = Divide(B=@n88.Absorption)
  n143 = StaticSwitchParameter(True=LinearInterpolate(A=@n142, B=@n134, Alpha=@n135), False=@n142) {parameterName=Use_JungleAlgea}
  n144 = Multiply(A=@n138, B=FeatureLevelSwitch(Default=QualitySwitch(Default=@n143, Low=@n142, High=@n143, Medium=@n143, Epic=@n143), ES3_1=@n142))
  SingleLayerWaterMaterialOutput!(
   ScatteringCoefficients=StaticSwitchParameter(True=LinearInterpolate(A=@n133, B=@n139, Alpha=@n140), False=@n139) {parameterName=UseBoxBlendFlattenColor}
   AbsorptionCoefficients=StaticSwitchParameter(True=LinearInterpolate(A=@n141, B=@n144, Alpha=@n140), False=@n144) {parameterName=UseBoxBlendFlattenColor}
   PhaseG=CollectionParameter() {collection=&MPC_Water, parameterName=Ansotropy}
   ColorScaleBehindWater
  )
  StaticSwitchParameter(True=@n20.Foam Mask, False=Constant()) {parameterName=Debug Fluid Forces}
 ]
}
